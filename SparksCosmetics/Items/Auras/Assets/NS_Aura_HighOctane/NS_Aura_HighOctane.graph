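# Niagara system for the HighOctane cosmetic aura, wrapping a skeletal mesh in glow, flames and sparks.
# BodyGlow holds six bone-attached sprites; FlamesSprites spawns multicolored vertical flames at bone offsets with interpolated positions, while its mesh variant is disabled.
# GPU sparks launch upward between companion bone pairs, with spawn rate and size driven by LOD distance.
# User booleans IsLargeBody and IsFrontEnd adjust scale, offsets and brightness.
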NiagaraSystem /SparksCosmetics/Items/Auras/Assets/NS_Aura_HighOctane/NS_Aura_HighOctane
dependencies {
 usedDataInterfaces [
  &NiagaraDataInterfaceArrayInt32
  &NiagaraDataInterfaceArrayFloat
  &NiagaraDataInterfaceCurve
  &NiagaraDataInterfaceArrayFloat3
  &NiagaraDataInterfaceColorCurve
  &NiagaraDataInterfaceArrayColor
  &NiagaraDataInterfaceVector2DCurve
  &NiagaraDataInterfaceSkeletalMesh
 ]
 usedDynamicInputs []
 usedModules [
  &self:SystemSpawnScript.SystemScriptSource.SystemScriptGraph.NiagaraNodeAssignment_0.SetVariables_#1
  &self:SystemSpawnScript.SystemScriptSource.SystemScriptGraph.NiagaraNodeAssignment_4.SetVariables_#2
  &SystemState
  &BlendSpaceHelper
 ]
}
emitters {
 BodyGlow {
  emitterUpdateScript [
   &EmitterState()
   &SpawnBurst_Instantaneous(Spawn Count=6, Use Spawn Probability=false, UseLoopCountLimit=true)
  ]
  particleSpawnScript [
   &PartitionParticles(Number Of Partitions=6)
   &InitializeParticle(
    Color=&MakeCustomLinearColorFromBool(Bool=$User.IsLargeBody, True Color=<rgba 1 0.5061 0.1768 1>, False Color=<rgba 1 0.5061 0.1768 0.4>)
    Lifetime=&RandomRangeFloat(Minimum=3, Maximum=4)
    Sprite Rotation Mode=::NewEnumerator2
    Sprite Size Mode=::NewEnumerator3
    Uniform Sprite Size=&MakeCustomFloatFromBool(Bool=$User.IsLargeBody, True Float=30, False Float=20)
    UsePositionOffset=false
    Write Material Random=true
   ) {module=InitializeParticle001}
   /SetVariables(
    Particles.PickBone:NiagaraInt32=&SelectIntFromArray(Direct Array Index=$Particles.PartitionParticles.Partition, Int Selection Array=DataInterface() {propertyValues {IntData [1,2,3,4,3,4]}})
    Particles.SymmetricalLerp:NiagaraFloat=&Multiply_Float(
     A=/Niagara/DynamicInputs/Integer/V2/Round_Float(Float:NiagaraFloat=&Multiply_FloatByInt(Float=0.49, Integer=$Particles.PartitionParticles.Partition))
     B=0.5
    )
   ) {isSetParameters=true}
   /SetVariables(
    Particles.glowCameraOffset=-10
    Particles.glowColor:LinearColor=&MakeCustomLinearColorFromBool(
     Bool=$User.IsFrontEnd
     True Color=&LerpLinearColors(StartColor=<rgba 1 0.5061 0.1768 0.1>, EndColor=<rgba 1 0.2591 0.1768 0.05>, LerpFactor=$Particles.SymmetricalLerp)
     False Color=&LerpLinearColors(StartColor=<rgba 1 0.5061 0.1768 0.15>, EndColor=<rgba 1 0.2591 0.1768 0.05>, LerpFactor=$Particles.SymmetricalLerp)
    )
    Particles.glowScale:Vector2f=&Vector2DFromFloat(Value=&Multiply_Float(
      A=&SelectFloatFromArray(
       Direct Array Index=$Particles.PartitionParticles.Partition
       Float Selection Array=DataInterface() {propertyValues {FloatData [60,60,40,40,40,40]}}
      )
      B=&MakeCustomFloatFromBool(Bool=$User.IsLargeBody, True Float=1.3, False Float=1)
     ))
   ) {isSetParameters=true}
  ]
  particleUpdateScript [
   &ParticleState(Kill Particles When Lifetime Has Elapsed=false, Let Infinitely Lived Particles Die When Emitter Deactivates=true)
   &CameraOffset(Camera Offset Amount=5)
   &ScaleSpriteSize(
    Scale Sprite Size Mode=::NewEnumerator0
    Uniform Scale Factor=&Lerp_Float(A=1.5, B=1, Alpha=$Particles.SymmetricalLerp)
   )
   &ScaleColor(
    Scale RGB=&VectorFromFloat(Value=&Lerp_Float(A=3, B=0.75, Alpha=$Particles.SymmetricalLerp))
    ScaleA=true
    ScaleRGB=true
   )
   &SpriteRotationRate(Rotation Rate=&RandomRangeFloat(Minimum=10, Maximum=20))
   &SkeletalMeshLocation(
    Bone / Socket Index=$Particles.PickBone
    Bone Sampling Mode=::NewEnumerator4
    MaskSpawnGroup=false
    Orientation Sampling=::NewEnumerator1
    Position Sampling=::NewEnumerator2
    Sampled Position Offset=&Multiply_VectorByFloat(
     Vector=&SelectVectorFromArray(
      Direct Array Index=$Particles.PickBone
      Vector Selection Array=DataInterface() {propertyValues {FloatData [<0 0 0>, <-5 5 0>, <5 -5 0>, <0 0 0>]}}
     )
     Float=&MakeCustomFloatFromBool(Bool=$User.IsLargeBody, True Float=1, False Float=0.5)
    )
    Skeletal Mesh=$System.Skeletal Mesh
   )
   &SkeletalMeshLocation(
    Bone / Socket Index=&SelectIntFromArray(Direct Array Index=$Particles.PickBone, Int Selection Array=$System.CompanionBone)
    Bone Sampling Mode=::NewEnumerator4
    MaskSpawnGroup=false
    Orientation Sampling=::NewEnumerator1
    Position Sampling=::NewEnumerator2
    Sampled Position Offset Coordinate Space:ENiagaraSkelMeshTransforms=::NewEnumerator1
    Skeletal Mesh=$System.Skeletal Mesh
   ) {module=SkeletalMeshLocation001}
   &DynamicMaterialParameters(
    Index 0 Param 1=&SelectFloatFromArray(
     Direct Array Index=$Particles.PartitionParticles.Partition
     Float Selection Array=DataInterface() {propertyValues {FloatData [5,5,5,4.5,4.5,4,4,3,3]}}
    )
    Index 0 Param 3=1
    Module.Param0WriteEnabled=true
    Module.Param1WriteEnabled=true
    Module.Param2WriteEnabled=true
    Module.Param3WriteEnabled=true
   )
   /SetVariables(Particles.Position=&LerpPosition(
     A=$Particles.SkeletalMeshLocation.SampledPosition
     B=$Particles.SkeletalMeshLocation001.SampledPosition
     Alpha=&SelectFloatFromArray(
      Direct Array Index=$Particles.PartitionParticles.Partition
      Float Selection Array=DataInterface() {propertyValues {FloatData [0,0,0.5,0.5,0,0,0.5,0.5,0,0]}}
     )
    )) {isSetParameters=true}
   &Color(Color=&ColorFromCurve(
     ColorCurve=DataInterface() {propertyValues {
       AlphaCurve {keys [<at 0 1 RCIM_Cubic>]}
       BlueCurve {keys [<at 0 0 RCIM_Cubic>,<at 0.5 0>,<at 1 1 RCIM_Cubic>]}
       GreenCurve {keys [<at 0 0 RCIM_Cubic>,<at 0.5 1>,<at 1 0 RCIM_Cubic>]}
       RedCurve {keys [<at 0 1 RCIM_Cubic>,<at 0.5 0>,<at 1 0 RCIM_Cubic>]}
      }}
     CurveIndex=$Particles.SymmetricalLerp
    )) {enabled=false}
  ]
  rendererClasses [&NiagaraSpriteRendererProperties]
  renderers [
   {rendererClass=&NiagaraSpriteRendererProperties}
   {rendererClass=&NiagaraSpriteRendererProperties, rendererIndex=1}
   {rendererClass=&NiagaraSpriteRendererProperties, rendererIndex=2}
  ]
 }
 FlamesMesh {
  bEnabled=false
  emitterUpdateScript [
   &EmitterState()
   &SpawnRate(Spawn Probability=0.6, SpawnRate=100, Use Spawn Probability=true)
   &SpawnRate(Spawn Probability=0.6, SpawnGroup=1, SpawnRate=50, Use Spawn Probability=true) {module=SpawnRate001}
   &SpawnBurst_Instantaneous(Spawn Count=10, Use Spawn Probability=false, UseLoopCountLimit=true) {enabled=false}
  ]
  particleSpawnScript [
   &PartitionParticles(
    Normalized Index In Partition Includes One=true
    Number Of Partitions=4
    Particles Per Partition=2
    Partition Mode:ENiagara_PartitionMode_v2=::NewEnumerator1
   )
   /SetVariables(Particles.SymmetricalLerp:NiagaraFloat=&Add_Float(
     A=&Multiply_Float(A=$Particles.PartitionParticles.Normalized Index In Partition, B=0.3333)
     B=&SelectFloatFromArray(
      Direct Array Index=$Particles.PartitionParticles.Partition
      Float Selection Array=DataInterface() {propertyValues {FloatData [0,0,0.6667,0.6667]}}
     )
    )) {isSetParameters=true, module=SetVariables_#3}
   &InitializeParticle(
    Color=&SelectLinearColorFromArray(
     Array Sampling Mode=::NewEnumerator0
     Direct Array Index=?
     Color Selection Array=DataInterface() {propertyValues {ColorData [<rgba 0.0511 1 0 1>, <rgba 0 0.792 1 1>, <rgba 1 0.317 0 1>]}}
     Random Seed=0
    )
    Lifetime=&MaskFloatBySpawnGroup(Masked Float=0.5, Passthrough Float=1)
    Material Random:NiagaraFloat=&RandomRangeFloat(Minimum=0, Maximum=1)
    Mesh Scale=&MakeVector(
     X=1
     Y=1
     Z=&Multiply_Float(
      A=&Multiply_Float(
       A=&Lerp_Float(A=1.35, B=0.75, Alpha=$Particles.SymmetricalLerp)
       B=&MaskFloatBySpawnGroup(Masked Float=1, Passthrough Float=1.25)
      )
      B=&MakeCustomFloatFromBool(Bool=$User.IsLargeBody, True Float=1.25, False Float=1)
     )
    )
    Mesh Scale Mode=::NewEnumerator4
    Sprite Size Mode=::NewEnumerator3
    Uniform Sprite Size=15
    UsePositionOffset=false
    Write Material Random=true
   )
   &CameraOffset(Camera Offset Amount=&MaskFloatBySpawnGroup(Masked Float=-5, Passthrough Float=-10))
   /SetVariables(Particles.VisibilityTag=&SpawnGroup()) {isSetParameters=true, module=SetVariables_#4}
  ]
  particleUpdateScript [
   &ParticleState()
   &ScaleMeshSize(Scale Factor=&VectorFromFloat(Value=&FloatFromCurve(FloatCurve=DataInterface() {propertyValues {Curve {defaultValue=1, keys [{tangentMode=RCTM_User, time=0, value=0}, {tangentMode=RCTM_User, time=1, value=1}]}}})))
   &UpdateMeshOrientation(
    Facing Direction=/Game/Effects/Niagara/DynamicInputs/Engine/FN_GetSystemAxis(Axis:ENiagaraOrientationAxis=::XAxis, Flip Axis:NiagaraBool=false)
    Facing Interpolation=false
    Orientation Method=::NewEnumerator1
    Override Quat=false
    TransformFacing=false
    TransformUp=false
   ) {enabled=false, module=UpdateMeshOrientation001}
   &SkeletalMeshLocation(
    Bone / Socket Index=&SelectIntFromArray(Direct Array Index=$Particles.PartitionParticles.Partition, Int Selection Array=DataInterface() {propertyValues {IntData [7,8,1,2]}})
    Bone Sampling Mode=::NewEnumerator4
    MaskSpawnGroup=false
    Orientation Sampling=::NewEnumerator1
    Sampled Position Offset=&Lerp_Vector(
     A=&SelectVectorFromArray(
      Direct Array Index=$Particles.PartitionParticles.Partition
      Vector Selection Array=DataInterface() {propertyValues {FloatData [<2 0 0>, <-2 0 0>, <10 0 0>, <-10 0 0>]}}
     )
     B=&SelectVectorFromArray(
      Direct Array Index=$Particles.PartitionParticles.Partition
      Vector Selection Array=DataInterface() {propertyValues {FloatData [<10 0 0>, <-10 0 0>, <20 0 0>, <-20 0 0>]}}
     )
     Alpha=&Add_Float(
      A=$Particles.PartitionParticles.Normalized Index In Partition
      B=&RandomRangeFloat(Minimum=-0.5, Maximum=0.2)
     )
    )
    Skeletal Mesh=$System.Skeletal Mesh
   )
   /SetVariables(Particles.Position=&AddVectorToPosition(
     Position=$Particles.Position
     Vector=&Multiply_VectorByFloat(
      Vector=&Lerp_Vector(A=<0 0 4>, B=<0 0 0>, Alpha=$Particles.SymmetricalLerp)
      Float=&MakeCustomFloatFromBool(Bool=$User.IsLargeBody, True Float=2, False Float=1)
     )
    )) {enabled=false, isSetParameters=true, module=SetVariables_#5}
   &InterpolateOverTime(Rate Of Change=15, Target Position Value:NiagaraPosition=$Particles.Position, Variable Type=::NewEnumerator6)
   /SetVariables(Particles.Interpolate Dynamic=&MakeFloat4FromVectorAndFloat("Vector (RGB)"=$Particles.InterpolateOverTime.InterpolateOverTime_Position.Moving Average, "Float (Alpha)"=1)) {isSetParameters=true}
   &Color(Color=&ColorFromCurve(
     ColorCurve=DataInterface() {propertyValues {
       AlphaCurve {keys [<at 0 1 RCIM_Cubic>]}
       BlueCurve {keys [<at 0 0 RCIM_Cubic>,<at 0.5 0>,<at 1 1 RCIM_Cubic>]}
       GreenCurve {keys [<at 0 0 RCIM_Cubic>,<at 0.5 1>,<at 1 0 RCIM_Cubic>]}
       RedCurve {keys [<at 0 1 RCIM_Cubic>,<at 0.5 0>,<at 1 0 RCIM_Cubic>]}
      }}
     CurveIndex=$Particles.SymmetricalLerp
    )) {enabled=false}
   &DynamicMaterialParameters(
    Index 0 Param 1=&FloatFromCurve(
     FloatCurve=DataInterface() {propertyValues {Curve {
        defaultValue=1
        keys [
         {interpMode=RCIM_Cubic, leaveTangentWeight=0.7868, tangentMode=RCTM_User, time=0, value=2}
         {arriveTangent=12.1916, arriveTangentWeight=0.7868, interpMode=RCIM_Cubic, leaveTangent=12.1916, tangentMode=RCTM_User, time=1, value=8}
        ]
       }}}
     CurveIndex=&RandomRangeFloat(Minimum=0, Maximum=1)
     Scale Curve=&MaskFloatBySpawnGroup(
      Masked Float=1
      Passthrough Float=&FloatFromCurve(
       FloatCurve=DataInterface() {propertyValues {Curve {keys [<at 0 1 RCIM_Cubic>, {arriveTangent=-3.0202, tangentMode=RCTM_User, time=1, value=0}]}}}
       CurveIndex=$Particles.SymmetricalLerp
       Scale Curve=2
      )
     )
    )
    Index 0 Param 2=&RandomRangeFloat(Minimum=0, Maximum=1)
    Index 0 Param 3=&RandomRangeFloat(Minimum=0, Maximum=1)
    Index 0 Param 4=&FloatFromCurve(FloatCurve=DataInterface() {propertyValues {Curve {keys [{arriveTangent=-13.6913, interpMode=RCIM_Cubic, leaveTangent=-13.6913, tangentMode=RCTM_User, time=0, value=5}, <at 1 1>]}}})
    Module.Param0WriteEnabled=true
    Module.Param1WriteEnabled=true
    Module.Param2WriteEnabled=true
    Module.Param3WriteEnabled=true
   )
   &ScaleColor(
    Scale Alpha=&FloatFromCurve(
     FloatCurve=DataInterface() {propertyValues {Curve {keys [
         {arriveTangent=5.0769, interpMode=RCIM_Cubic, leaveTangent=5.0769, tangentMode=RCTM_User, time=0, value=0}
         <at 0.25 1 RCIM_Cubic>
         {arriveTangent=-1.9131, tangentMode=RCTM_User, time=1, value=0}
        ]}}}
     Scale Curve=&MaskFloatBySpawnGroup(Masked Float=1, Passthrough Float=4)
    )
    Scale RGB=&VectorFromFloat(Value=&FloatFromCurve(
      FloatCurve=DataInterface() {propertyValues {Curve {keys [{arriveTangent=-11.8698, interpMode=RCIM_Cubic, leaveTangent=-11.8698, tangentMode=RCTM_User, time=0, value=5}, <at 1 1>]}}}
      Scale Curve=&MaskFloatBySpawnGroup(Masked Float=1, Passthrough Float=0.05)
     ))
    ScaleA=true
    ScaleRGB=true
   )
  ]
  rendererClasses [&NiagaraMeshRendererProperties, &NiagaraSpriteRendererProperties]
  renderers [
   {rendererClass=&NiagaraMeshRendererProperties, rendererIndex=0}
   {rendererClass=&NiagaraMeshRendererProperties, rendererIndex=1}
   {rendererClass=&NiagaraSpriteRendererProperties, rendererIndex=2}
  ]
 }
 FlamesSprites {
  emitterUpdateScript [
   &EmitterState()
   &SpawnRate(Spawn Probability=0.6, SpawnRate=100, Use Spawn Probability=true)
   &SpawnRate(Spawn Probability=0.6, SpawnGroup=1, SpawnRate=50, Use Spawn Probability=true) {module=SpawnRate001}
   &SpawnBurst_Instantaneous(Spawn Count=10, Use Spawn Probability=false, UseLoopCountLimit=true) {enabled=false}
  ]
  particleSpawnScript [
   &PartitionParticles(
    Normalized Index In Partition Includes One=true
    Number Of Partitions=4
    Particles Per Partition=2
    Partition Mode:ENiagara_PartitionMode_v2=::NewEnumerator1
   )
   /SetVariables(Particles.SymmetricalLerp:NiagaraFloat=&Add_Float(
     A=&Multiply_Float(A=$Particles.PartitionParticles.Normalized Index In Partition, B=0.3333)
     B=&SelectFloatFromArray(
      Direct Array Index=$Particles.PartitionParticles.Partition
      Float Selection Array=DataInterface() {propertyValues {FloatData [0,0,0.6667,0.6667]}}
     )
    )) {isSetParameters=true, module=SetVariables_#3}
   &InitializeParticle(
    Color=&SelectLinearColorFromArray(
     Array Sampling Mode=::NewEnumerator0
     Direct Array Index=?
     Color Selection Array=DataInterface() {propertyValues {ColorData [<rgba 0.0511 1 0 1>, <rgba 0 0.792 1 1>, <rgba 1 0.317 0 1>]}}
     Random Seed=0
    )
    Lifetime=&MaskFloatBySpawnGroup(Masked Float=0.5, Passthrough Float=1)
    Material Random:NiagaraFloat=&RandomRangeFloat(Minimum=0, Maximum=1)
    Sprite Size Max=<1 60>
    Sprite Size Min=<1 50>
    Sprite Size Mode=::NewEnumerator2
    UsePositionOffset=false
    Write Material Random=true
   )
   &CameraOffset(Camera Offset Amount=&MaskFloatBySpawnGroup(Masked Float=0, Passthrough Float=-10))
   /SetVariables(
    Transient.WidthByPosition:NiagaraFloat=&MaskFloatBySpawnGroup(
     Masked Float=1
     Passthrough Float=&FloatFromCurve(
      FloatCurve=DataInterface() {propertyValues {Curve {keys [<at 0 1 RCIM_Cubic>, {arriveTangent=-3.0202, tangentMode=RCTM_User, time=1, value=0}]}}}
      CurveIndex=$Particles.SymmetricalLerp
      Scale Curve=2
     )
    )
    Transient.WidthDistribution:NiagaraFloat=&FloatFromCurve(
     FloatCurve=DataInterface() {propertyValues {Curve {
        defaultValue=1
        keys [
         {arriveTangent=3.0702, interpMode=RCIM_Cubic, leaveTangent=3.0702, leaveTangentWeight=0.7868, tangentMode=RCTM_User, time=0, value=2}
         {arriveTangentWeight=0.527, interpMode=RCIM_Cubic, leaveTangentWeight=0.527, tangentMode=RCTM_User, time=0.5, value=3.5}
         {arriveTangent=5.7817, arriveTangentWeight=0.7868, interpMode=RCIM_Cubic, leaveTangent=5.7817, tangentMode=RCTM_User, time=1, value=5}
        ]
       }}}
     CurveIndex=&RandomRangeFloat(Minimum=0, Maximum=1)
    )
    Transient.WidthLargeBody:NiagaraFloat=&MakeCustomFloatFromBool(Bool=$User.IsLargeBody, True Float=2, False Float=1)
   ) {isSetParameters=true}
   /SetVariables(
    Transient.LengthByColor:NiagaraFloat=&MaskFloatBySpawnGroup(Masked Float=1, Passthrough Float=1.25)
    Transient.LengthByPosition:NiagaraFloat=&Lerp_Float(A=1.35, B=0.75, Alpha=$Particles.SymmetricalLerp)
    Transient.LengthLargeBody:NiagaraFloat=&MakeCustomFloatFromBool(Bool=$User.IsLargeBody, True Float=1.25, False Float=1)
   ) {isSetParameters=true}
   /SetVariables(Particles.SpriteSize=&Multiply_Vector2D(
     A=$Particles.SpriteSize
     B=&MakeVector2D(
      X=&Multiply_Float(A=$Transient.WidthDistribution, B=&Multiply_Float(A=$Transient.WidthLargeBody, B=$Transient.WidthByPosition))
      Y=&Multiply_Float(A=&Multiply_Float(A=$Transient.LengthByPosition, B=$Transient.LengthByColor), B=$Transient.LengthLargeBody)
     )
    )) {isSetParameters=true}
   /SetVariables(Particles.VisibilityTag=&SpawnGroup()) {isSetParameters=true, module=SetVariables_#4}
  ]
  particleUpdateScript [
   &ParticleState()
   /SetVariables(Transient.HeightScale:NiagaraFloat=&FloatFromCurve(FloatCurve=DataInterface() {propertyValues {Curve {keys [<at 0 0>,<at 1 1>]}}})) {isSetParameters=true}
   &ScaleSpriteSize(Scale Factor=&MakeVector2D(X=1, Y=$Transient.HeightScale), Scale Sprite Size Mode=::NewEnumerator2)
   &SkeletalMeshLocation(
    Bone / Socket Index=&SelectIntFromArray(Direct Array Index=$Particles.PartitionParticles.Partition, Int Selection Array=DataInterface() {propertyValues {IntData [7,8,1,2]}})
    Bone Sampling Mode=::NewEnumerator4
    MaskSpawnGroup=false
    Orientation Sampling=::NewEnumerator1
    Sampled Position Offset=&Add_Vector(
     A=&Multiply_Vector(
      A=&RandomVector(Vector Scale=3)
      B=&MakeCustomVectorFromBool(Bool=$User.IsLargeBody, True Vector=<2 2 1>, False Vector=<1 1 0.25>)
     )
     B=&Lerp_Vector(
      A=&SelectVectorFromArray(
       Direct Array Index=$Particles.PartitionParticles.Partition
       Vector Selection Array=DataInterface() {propertyValues {FloatData [<2 0 0>, <-2 0 0>, <10 0 0>, <-10 0 0>]}}
      )
      B=&SelectVectorFromArray(
       Direct Array Index=$Particles.PartitionParticles.Partition
       Vector Selection Array=DataInterface() {propertyValues {FloatData [<10 0 0>, <-10 0 0>, <20 0 0>, <-20 0 0>]}}
      )
      Alpha=&Add_Float(
       A=$Particles.PartitionParticles.Normalized Index In Partition
       B=&RandomRangeFloat(Minimum=-0.5, Maximum=0.2)
      )
     )
    )
    Skeletal Mesh=$System.Skeletal Mesh
   )
   /SetVariables(Particles.Position=&AddVectorToPosition(
     Position=$Particles.Position
     Vector=&MakeCustomVectorFromBool(
      Bool=$User.IsLargeBody
      True Vector=&Lerp_Vector(A=<0 0 8>, B=<0 0 0>, Alpha=$Particles.SymmetricalLerp)
      False Vector=<0 0 0>
     )
    )) {isSetParameters=true, module=SetVariables_#5}
   &InterpolateOverTime(Rate Of Change=15, Target Position Value:NiagaraPosition=$Particles.Position, Variable Type=::NewEnumerator6)
   &Color(Color=&ColorFromCurve(
     ColorCurve=DataInterface() {propertyValues {
       AlphaCurve {keys [<at 0 1 RCIM_Cubic>]}
       BlueCurve {keys [<at 0 0 RCIM_Cubic>,<at 0.5 0>,<at 1 1 RCIM_Cubic>]}
       GreenCurve {keys [<at 0 0 RCIM_Cubic>,<at 0.5 1>,<at 1 0 RCIM_Cubic>]}
       RedCurve {keys [<at 0 1 RCIM_Cubic>,<at 0.5 0>,<at 1 0 RCIM_Cubic>]}
      }}
     CurveIndex=$Particles.SymmetricalLerp
    )) {enabled=false}
   &DynamicMaterialParameters(
    Index 0 Param 2=&RandomRangeFloat(Minimum=0, Maximum=1)
    Index 0 Param 3=&RandomRangeFloat(Minimum=0, Maximum=1)
    Index 0 Param 4=&FloatFromCurve(FloatCurve=DataInterface() {propertyValues {Curve {keys [{arriveTangent=-13.6913, interpMode=RCIM_Cubic, leaveTangent=-13.6913, tangentMode=RCTM_User, time=0, value=5}, <at 1 1>]}}})
    Module.Param0WriteEnabled=false
    Module.Param1WriteEnabled=true
    Module.Param2WriteEnabled=true
    Module.Param3WriteEnabled=true
   )
   &ScaleColor(
    Scale Alpha=&FloatFromCurve(
     FloatCurve=DataInterface() {propertyValues {Curve {keys [
         {arriveTangent=5.0769, interpMode=RCIM_Cubic, leaveTangent=5.0769, tangentMode=RCTM_User, time=0, value=0}
         <at 0.25 1 RCIM_Cubic>
         {arriveTangent=-1.9131, tangentMode=RCTM_User, time=1, value=0}
        ]}}}
     Scale Curve=&MaskFloatBySpawnGroup(Masked Float=1, Passthrough Float=4)
    )
    Scale RGB=&VectorFromFloat(Value=&FloatFromCurve(
      FloatCurve=DataInterface() {propertyValues {Curve {keys [{arriveTangent=-11.8698, interpMode=RCIM_Cubic, leaveTangent=-11.8698, tangentMode=RCTM_User, time=0, value=5}, <at 1 1>]}}}
      Scale Curve=&Multiply_Float(
       A=&MaskFloatBySpawnGroup(Masked Float=1, Passthrough Float=0.05)
       B=&MakeCustomFloatFromBool(Bool=$User.IsFrontEnd, True Float=1, False Float=2)
      )
     ))
    ScaleA=true
    ScaleRGB=true
   )
   &SpriteFacingAndAlignment(Alignment Coordinate Space=::World, Sprite Alignment=<0 0 1>, WriteAlignment=true, WriteFacing=false)
   /SetVariables(
    Particles.TopColor:LinearColor=&Multiply_LinearColorByFloat(
     Linear Color=$Particles.Initial.Color
     Float=&FloatFromCurve(FloatCurve=DataInterface() {propertyValues {Curve {
         defaultValue=0
         keys [
          {interpMode=RCIM_Cubic, tangentMode=RCTM_User, time=0, value=0}
          {interpMode=RCIM_Cubic, tangentMode=RCTM_User, time=0.5, value=1}
          {tangentMode=RCTM_User, time=1, value=0}
         ]
        }}})
    )
    Particles.TopFacing:Vector3f=<0 0 1>
    Particles.TopOffset=1
    Particles.TopPosition:NiagaraPosition=&AddVectorToPosition(Position=$Particles.Position, Vector=&Lerp_Vector(A=<0 0 6>, B=<0 0 12>, Alpha=$Transient.HeightScale))
    Particles.TopScale:Vector2f=&Multiply_Vector2D_ByFloat(Float=&MakeFloatFromVector2d(Vector 2D=$Particles.SpriteSize), Vector2D=<1 0.5>)
   ) {isSetParameters=true}
  ]
  rendererClasses [&NiagaraSpriteRendererProperties]
  renderers [
   {rendererClass=&NiagaraSpriteRendererProperties}
   {rendererClass=&NiagaraSpriteRendererProperties, rendererIndex=1}
   {rendererClass=&NiagaraSpriteRendererProperties, rendererIndex=2}
   {rendererClass=&NiagaraSpriteRendererProperties, rendererIndex=3}
  ]
 }
 NE_FN_Sparks {
  simTarget=GPUComputeSim
  emitterUpdateScript [
   &EmitterState()
   &SpawnRate(
    SpawnRate=&FloatFromCurve(
     FloatCurve=DataInterface() {propertyValues {Curve {keys [
         {arriveTangent=-0.0017, interpMode=RCIM_Cubic, leaveTangent=-0.0017, tangentMode=RCTM_User, time=300, value=10}
         {arriveTangent=-0.0055, interpMode=RCIM_Cubic, leaveTangent=-0.0055, tangentMode=RCTM_User, time=1412.7, value=6}
         {arriveTangent=-0.0006, tangentMode=RCTM_User, time=2500, value=1}
        ]}}}
     CurveIndex=$Engine.Owner.LODDistance
     Scale Curve=&RandomRangeFloat(Minimum=10, Maximum=20, Evaluation Type=::NewEnumerator1)
    )
    Use Spawn Probability=false
   )
  ]
  particleSpawnScript [
   /SetVariables(
    Particles.PickBone:NiagaraInt32=&RandomRangeInt(Minimum=1, Maximum=4)
    Particles.SparkIntensity:NiagaraFloat=&FloatFromCurve(
     FloatCurve=DataInterface() {propertyValues {Curve {keys [
         <at 0 1 RCIM_Cubic>
         {arriveTangent=-2.5317, interpMode=RCIM_Cubic, leaveTangent=-2.5317, tangentMode=RCTM_User, time=0.6, value=0.5}
         <at 1 0 RCIM_Cubic>
        ]}}}
     CurveIndex=&RandomRangeFloat(Minimum=0, Maximum=1)
    )
   ) {isSetParameters=true}
   &InitializeParticle(
    Color=&MultiplyLinearColor(
     Linear Color=&SelectLinearColorFromArray(
      Array Sampling Mode=::NewEnumerator0
      Direct Array Index=?
      Color Selection Array=DataInterface() {propertyValues {ColorData [<rgba 0.0511 1 0 1>, <rgba 0 0.792 1 1>, <rgba 1 0.317 0 1>]}}
      Random Seed=0
     )
     Scale Factor=/Niagara/DynamicInputs/TypeConversions/MakeCustomVector4FromBool(Bool:NiagaraBool=$User.IsFrontEnd, True Vector4:Vector4f=<5 5 5 1>, False Vector4:Vector4f=<1 1 1 1>)
    )
    Lifetime=&Lerp_Float(
     A=&RandomRangeFloat(Minimum=0.15, Maximum=0.2)
     B=&RandomRangeFloat(Minimum=0.25, Maximum=0.35)
     Alpha=$Particles.SparkIntensity
    )
    Mass=&Lerp_Float(A=0.3, B=1, Alpha=$Particles.SparkIntensity)
    Mass Mode=::NewEnumerator1
    Sprite Size Mode=::NewEnumerator3
    Uniform Sprite Size=&Lerp_Float(A=0.45, B=0.9, Alpha=$Particles.SparkIntensity)
    UsePositionOffset=false
    Write Material Random=true
   )
   /SetVariables(Particles.Color=&RandomRangeLinearColor(Minimum=<rgba 1 0.4064 0.08 0.5>, Maximum=<rgba 3 2.6103 0.3 1>)) {enabled=false, isSetParameters=true}
   &SkeletalMeshLocation(
    Bone / Socket Index=$Particles.PickBone
    Bone Sampling Mode=::NewEnumerator4
    MaskSpawnGroup=false
    Orientation Sampling=::NewEnumerator1
    Position Sampling=::NewEnumerator2
    Sampled Position Offset=&MakeCustomVectorFromBool(
     Bool=$User.IsLargeBody
     True Vector=&SelectVectorFromArray(Direct Array Index=$Particles.PickBone, Vector Selection Array=$System.BoneOffset)
     False Vector=<0 0 0>
    )
    Skeletal Mesh=$System.Skeletal Mesh
   ) {module=SkeletalMeshLocation001}
   &SkeletalMeshLocation(
    Bone / Socket Index=&SelectIntFromArray(Direct Array Index=$Particles.PickBone, Int Selection Array=$System.CompanionBone)
    Bone Sampling Mode=::NewEnumerator4
    MaskSpawnGroup=false
    Orientation Sampling=::NewEnumerator1
    Position Sampling=::NewEnumerator2
    Sampled Position Offset=&MakeCustomVectorFromBool(
     Bool=$User.IsLargeBody
     True Vector=&SelectVectorFromArray(
      Direct Array Index=&SelectIntFromArray(Direct Array Index=$Particles.PickBone, Int Selection Array=$System.CompanionBone)
      Vector Selection Array=$System.BoneOffset
     )
     False Vector=<0 0 0>
    )
    Skeletal Mesh=$System.Skeletal Mesh
   )
   &ShapeLocation(
    MaskSpawnGroup=false
    Override Seed=false
    Shape Origin=&LerpPosition(
     A=$Particles.SkeletalMeshLocation001.SampledPosition
     B=$Particles.SkeletalMeshLocation.SampledPosition
     Alpha=&RandomRangeFloat(Minimum=0, Maximum=1)
    )
    Sphere Radius=1
    UseOwnerScale=true
   )
   &AddVelocity(
    Velocity=<0 0 1>
    Velocity Speed Scale=&Multiply_Float(
     A=&RandomRangeFloat(Minimum=1, Maximum=2.5)
     B=&MakeCustomFloatFromBool(Bool=$User.IsLargeBody, True Float=100, False Float=80)
    )
   )
   &CameraOffset(Camera Offset Amount=-5)
  ]
  particleUpdateScript [
   &ParticleState()
   &SpriteRotationRate(Rotation Rate=&RandomRangeFloat(Minimum=-200, Maximum=200)) {enabled=false}
   &GravityForce(Gravity=<0 0 -100>)
   &Drag(Use Linear Drag=true, Use Rotational Drag=false)
   &VectorNoiseForce(
    Force Amount=&FloatFromCurve(
     FloatCurve=DataInterface() {propertyValues {Curve {keys [
         {arriveTangent=-9.4515, interpMode=RCIM_Cubic, leaveTangent=-4.1687, tangentMode=RCTM_User, time=0, value=1}
         {arriveTangent=-3.4372, interpMode=RCIM_Cubic, leaveTangent=-3.4372, tangentMode=RCTM_User, time=0.6, value=0}
        ]}}}
     Scale Curve=500
    )
    Module.Use Random Noise Time Interval=false
   ) {enabled=false}
   &SolveForcesAndVelocity(Clamp Velocity=false, Limit Acceleration=false)
   &ScaleColor(
    Linear Color Curve=DataInterface() {propertyValues {
      AlphaCurve {keys [<at 0 1>,<at 1 1>]}
      BlueCurve {keys [
        {arriveTangent=-36.8273, interpMode=RCIM_Cubic, leaveTangent=-36.8273, tangentMode=RCTM_User, time=0, value=5}
        {arriveTangent=-1.452, interpMode=RCIM_Cubic, leaveTangent=-1.452, tangentMode=RCTM_User, time=0.163, value=1}
        <at 1 0.1357>
       ]}
      GreenCurve {keys [
        {arriveTangent=-36.8273, interpMode=RCIM_Cubic, leaveTangent=-36.8273, tangentMode=RCTM_User, time=0, value=5}
        {arriveTangent=-1.4408, interpMode=RCIM_Cubic, leaveTangent=-1.4408, tangentMode=RCTM_User, time=0.163, value=1}
        <at 1 0.1357>
       ]}
      RedCurve {keys [
        {arriveTangent=-36.8273, interpMode=RCIM_Cubic, leaveTangent=-36.8273, tangentMode=RCTM_User, time=0, value=5}
        {arriveTangent=-0.8332, interpMode=RCIM_Cubic, leaveTangent=-0.8332, tangentMode=RCTM_User, time=0.163, value=1}
        <at 1 0.1357>
       ]}
     }}
    Scale Mode=::NewEnumerator2
   )
   &ScaleSpriteSize(Scale Factor=&Vector2DFromFloat(Value=&FloatFromCurve(FloatCurve=DataInterface() {propertyValues {Curve {keys [
          {arriveTangent=-0.8333, interpMode=RCIM_Cubic, time=0, value=0}
          {arriveTangentWeight=0.1795, interpMode=RCIM_Cubic, leaveTangentWeight=0.1795, tangentMode=RCTM_User, time=0.2, value=1}
          {arriveTangent=-0.625, interpMode=RCIM_Cubic, leaveTangent=-0.625, time=0.4, value=1}
          {arriveTangent=-2.2814, interpMode=RCIM_Cubic, leaveTangent=-2.2814, tangentMode=RCTM_User, time=1, value=0.5}
         ]}}})))
   &ScaleSpriteSizeBySpeed(
    Max Scale Factor=<1.25 8>
    Sample Scale Factor By Curve=true
    Scale Factor Curve=DataInterface() {propertyValues {Curve {keys [
        {arriveTangent=0.2757, time=0, value=0}
        {arriveTangent=0.3844, interpMode=RCIM_Cubic, leaveTangent=0.3844, tangentMode=RCTM_User, time=0.3, value=0.15}
        {arriveTangent=2.2765, interpMode=RCIM_Cubic, leaveTangent=2.2765, tangentMode=RCTM_User, time=1, value=1}
       ]}}}
    Velocity Threshold:NiagaraFloat=&FloatFromCurve(FloatCurve=DataInterface() {propertyValues {Curve {keys [<at 0.25 30>,<at 1 100>]}}})
   )
   /SetVariables(Particles.DistScale:Vector2f=&Multiply_Vector2(
     A=$Particles.SpriteSize
     B=&Vector2DFromFloat(Value=&FloatFromCurve(
       FloatCurve=DataInterface() {propertyValues {Curve {keys [
           {arriveTangent=0.0008, interpMode=RCIM_Cubic, leaveTangent=0.0008, leaveTangentWeight=548.2426, tangentMode=RCTM_User, time=300, value=1}
           {
            arriveTangent=0.0027
            arriveTangentWeight=548.25
            interpMode=RCIM_Cubic
            leaveTangent=0.0027
            leaveTangentWeight=185.1056
            tangentMode=RCTM_User
            time=1944.7
            value=3
           }
           {arriveTangent=0.0073, arriveTangentWeight=185.1033, interpMode=RCIM_Cubic, leaveTangent=0.0073, tangentMode=RCTM_User, time=2500, value=6}
          ]}}}
       CurveIndex=$Engine.Owner.LODDistance
       Scale Curve=&RandomRangeFloat(Minimum=0.5, Maximum=1)
      ))
    )) {isSetParameters=true}
   &BlendSpace(Rotation Contribution=0.5)
  ]
  rendererClasses [&NiagaraSpriteRendererProperties]
  renderers [{rendererClass=&NiagaraSpriteRendererProperties}]
 }
}
summary {
 systemName=NS_Aura_HighOctane
 emitters [
  {emitterName=BodyGlow, rendererClasses [&NiagaraSpriteRendererProperties]}
  {emitterName=NE_FN_Sparks, simTarget=GPUComputeSim, rendererClasses [&NiagaraSpriteRendererProperties]}
  {
   bEnabled=false
   emitterName=FlamesMesh
   rendererClasses [&NiagaraMeshRendererProperties, &NiagaraSpriteRendererProperties]
  }
  {emitterName=FlamesSprites, rendererClasses [&NiagaraSpriteRendererProperties]}
 ]
 userVariables [
  User.Interpolate Intensity {type=NiagaraFloat, defaultValue=3, description=""}
  User.IsFrontEnd {type=NiagaraBool}
  User.IsLargeBody {type=NiagaraBool}
 ]
}
